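# EQS query that picks a summon location for the DADBRO AI encounter around the current safe zone center.
# Grid points are projected, then filtered by minimum distance from center and building rifts, a collision-channel projection and a flat-surface test.
# Paired dot tests favor the side of center opposite the encounter actor while penalizing near-exact alignment.
# A negatively weighted distance score penalizes outlying points.
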EnvQuery /Game/AIDirector/EQS/EQS_EncounterDADBROSummonLocation
options [{
  generator {
   name=EnvQueryGenerator_SimpleGrid_1
   class=EnvQueryGenerator_SimpleGrid
   properties {
    generateAround=/Script/FortniteAIServer.FortQueryContext_AthenaCurrentSafeZoneCenter
    gridSize=6000
    spaceBetween=1200
    projectionData {
     extentX=50
     navigationFilter=&RecastFilter_UseDefaultArea
     postProjectionVerticalOffset=0
     projectDown=3000
     projectUp=10000
     serializedChannel=ECC_GameTraceChannel11
     traceChannel=TraceTypeQuery10
    }
   }
  }
  tests [
   EnvQueryTest_Dot_3 {
    class=EnvQueryTest_Dot
    properties {
     clampMaxType=SpecifiedValue
     clampMinType=SpecifiedValue
     floatValueMax=1
     floatValueMin=0.6
     scoreClampMax=1
     scoreClampMin=0.6
     scoringFactor=5
     testComment=""
     lineA {
      lineFrom=/Script/FortniteAIServer.FortQueryContext_EncounterQueryActor
      lineTo=&FortQueryContext_AthenaCurrentSafeZoneCenter
     }
     lineB {lineFrom=&FortQueryContext_AthenaCurrentSafeZoneCenter}
    }
   }
   EnvQueryTest_Dot_4 {
    class=EnvQueryTest_Dot
    properties {
     clampMaxType=SpecifiedValue
     clampMinType=SpecifiedValue
     scoreClampMax=1
     scoreClampMin=0.9
     scoringFactor=-10
     testComment=""
     lineA {
      lineFrom=&FortQueryContext_EncounterQueryActor
      lineTo=&FortQueryContext_AthenaCurrentSafeZoneCenter
     }
     lineB {lineFrom=&FortQueryContext_AthenaCurrentSafeZoneCenter}
    }
   }
   EnvQueryTest_Distance_9 {
    class=EnvQueryTest_Distance
    properties {
     clampMaxType=SpecifiedValue
     clampMinType=SpecifiedValue
     distanceTo=&FortQueryContext_AthenaCurrentSafeZoneCenter
     filterType=Minimum
     floatValueMin=1500
     scoreClampMax=1000
     scoringEquation=InverseLinear
     scoringFactor=-5
     testComment=""
     testMode=Distance2D
     testPurpose=Filter
    }
   }
   EnvQueryTest_Distance_10 {
    class=EnvQueryTest_Distance
    properties {
     clampMaxType=SpecifiedValue
     clampMinType=SpecifiedValue
     distanceTo=&FortQueryContext_AthenaCurrentSafeZoneCenter
     filterType=Range
     floatValueMin=0
     scoreClampMax=4000
     scoreClampMin=2500
     scoringEquation=Linear
     scoringFactor=-5
     testComment=""
     testMode=Distance2D
     testPurpose=Score
    }
   }
   EnvQueryTest_Distance_12 {
    class=EnvQueryTest_Distance
    properties {
     distanceTo=/Script/FortniteAIServer.FortQueryContext_BuildingRifts
     filterType=Minimum
     floatValueMin=500
     scoringEquation=Linear
     scoringFactor=1
     testComment=""
     testMode=Distance3D
     testPurpose=Filter
    }
   }
   EnvQueryTest_Project_3 {
    class=EnvQueryTest_Project
    properties {
     scoringFactor=1
     testComment=""
     testPurpose=Filter
     projectionData {
      bTraceComplex=true
      navigationFilter=null
      postProjectionVerticalOffset=5
      projectDown=115
      projectUp=115
      serializedChannel=ECC_GameTraceChannel6
      traceChannel=TraceTypeQuery7
      traceMode=GeometryByChannel
     }
    }
   }
   FortQueryTest_OnFlatSurface_3 {class=FortQueryTest_OnFlatSurface, properties {radius=50, toleranceZ=45, traceOffsetDown=250, traceOffsetUp=250}}
  ]
 }]
