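# An Environment Query in the AI Director folder that returns the querier's own location.
# Its single option uses a Current Location generator with the Querier context as its source, yielding that position as the query item.
# It has no tests and no projection tracing, so the point is neither scored, filtered nor snapped to navmesh or geometry.
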EnvQuery /Game/AIDirector/EQS/EQS_QuerierLocation
options [{
  generator {
   name=EnvQueryGenerator_CurrentLocation_0
   class=EnvQueryGenerator_CurrentLocation
   properties {
    bAutoSortTests=true
    bCanRunAsync=false
    envQueryResultNormalizationOption=Default
    navDataOverrideContext=null
    optionName=""
    queryContext=&EnvQueryContext_Querier
    projectionData {
     bCanDisableTrace=true
     bCanProjectDown=true
     bCanTraceOnGeometry=true
     bCanTraceOnNavMesh=true
     bOnlyBlockingHits=true
     bTraceComplex=false
     extentX=0
     extentY=10
     extentZ=10
     navigationFilter=null
     postProjectionVerticalOffset=0
     projectDown=1024
     projectUp=1024
     serializedChannel=ECC_Visibility
     traceChannel=TraceTypeQuery1
     traceMode=None
     traceProfileName=None
     traceShape=Line
     versionNum=1
    }
   }
  }
  tests []
 }]
