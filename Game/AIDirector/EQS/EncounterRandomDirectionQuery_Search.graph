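# EQS query for AI Director encounters that picks building rift locations near the target objective, biased toward a random encounter direction.
# Rift candidates from the BuildingRifts generator are filtered by a parameter-driven 3D distance range from the objective.
# A dot test with a parameter-driven minimum compares the random direction against the line from the goals' center, keeping one directional sector.
# Inverse-linear distance scoring favors rifts closer to the objective.
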EnvQuery /Game/AIDirector/EQS/EncounterRandomDirectionQuery_Search
options [{
  generator {
   name=FortQueryGenerator_BuildingRifts_0
   class=FortQueryGenerator_BuildingRifts
   properties {bAutoSortTests=true, bCanRunAsync=false, envQueryResultNormalizationOption=Default, optionName=""}
  }
  tests [
   EnvQueryTest_Distance_0 {
    class=EnvQueryTest_Distance
    properties {
     distanceTo=/Script/FortniteAIServer.FortQueryContext_EncounterTargetObjective
     filterType=Range
     scoringEquation=Linear
     scoringFactor=0
     testComment=""
     testMode=Distance3D
     testPurpose=Filter
     floatValueMax {
      dataBinding=&self:EnvQueryTest_Distance_0.AIDataProvider_QueryParams_0
      dataField=FloatValue
      defaultValue=0
     }
     floatValueMin {
      dataBinding=&self:EnvQueryTest_Distance_0.AIDataProvider_QueryParams_2
      dataField=FloatValue
      defaultValue=0
     }
    }
   }
   EnvQueryTest_Distance_1 {
    class=EnvQueryTest_Distance
    properties {
     distanceTo=&FortQueryContext_EncounterTargetObjective
     filterType=Range
     floatValueMin=0
     multipleContextFilterOp=AnyPass
     multipleContextScoreOp=MaxScore
     scoringEquation=InverseLinear
     scoringFactor=1
     testComment=""
     testMode=Distance3D
     testPurpose=Score
    }
   }
   EnvQueryTest_Dot_0 {
    class=EnvQueryTest_Dot
    properties {
     clampMaxType=None
     clampMinType=None
     filterType=Minimum
     multipleContextFilterOp=AnyPass
     multipleContextScoreOp=MaxScore
     scoreClampMax=0
     scoreClampMin=0
     scoringFactor=0
     testComment=""
     testPurpose=Filter
     floatValueMin {
      dataBinding=&self:EnvQueryTest_Dot_0.AIDataProvider_QueryParams_4
      dataField=FloatValue
      defaultValue=0
     }
     lineA {
      lineFrom=&FortQueryContext_EncounterTargetObjective
      rotation=/Script/FortniteAIServer.FortQueryContext_EncounterRandomDirection
     }
     lineB {lineFrom=/Script/FortniteAIServer.FortQueryContext_EncounterGoalsCenterLocation}
    }
   }
  ]
 }]
