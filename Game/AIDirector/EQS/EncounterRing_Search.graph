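# EQS query for the AI Director that searches for building rifts in a ring around an encounter's target objective.
# A BuildingRifts generator supplies the candidates, with async running disabled.
# A single 3D distance test to the objective filters them to a min/max range and scores nearer ones higher through inverse linear scoring.
# Range bounds and the score's upper clamp bind to query parameters, while the lower clamp follows the filter threshold.
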EnvQuery /Game/AIDirector/EQS/EncounterRing_Search
options [{
  generator {
   name=FortQueryGenerator_BuildingRifts_0
   class=FortQueryGenerator_BuildingRifts
   properties {bAutoSortTests=true, bCanRunAsync=false, envQueryResultNormalizationOption=Default, optionName=""}
  }
  tests [EnvQueryTest_Distance_1 {
    class=EnvQueryTest_Distance
    properties {
     clampMaxType=SpecifiedValue
     clampMinType=FilterThreshold
     distanceTo=/Script/FortniteAIServer.FortQueryContext_EncounterTargetObjective
     filterType=Range
     scoringEquation=InverseLinear
     scoringFactor=1
     testComment=""
     testMode=Distance3D
     testPurpose=FilterAndScore
     floatValueMax {
      dataBinding=&self:EnvQueryTest_Distance_1.AIDataProvider_QueryParams_1
      dataField=FloatValue
      defaultValue=0
     }
     floatValueMin {
      dataBinding=&self:EnvQueryTest_Distance_1.AIDataProvider_QueryParams_2
      dataField=FloatValue
      defaultValue=0
     }
     scoreClampMax {
      dataBinding=&self:EnvQueryTest_Distance_1.AIDataProvider_QueryParams_0
      dataField=FloatValue
      defaultValue=0
     }
    }
   }]
 }]
